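# Translucent unlit beam material for mesh particles that fattens with camera distance, via a vertex-normal offset scaled up toward SizeScale.
# An axis-aligned Fresnel blends inside and outside colors, with optional particle color and time-of-day switches.
# Opacity combines Fresnel, particle alpha, depth fade, near-camera fade and V-gradient end and top fades.
# Mobile paths substitute a flat color and plain particle alpha for the Fresnel and depth-fade terms.
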Material /Game/Effects/Fort_Effects/Materials/Beams/M_Beam_DistanceExpand
graph {
 properties {
  bAllowNegativeEmissiveColor=true
  bUsedWithMeshParticles=true
  bUsedWithNiagaraMeshParticles=true
  blendMode=BLEND_Translucent
  pixelDepthOffsetMode=PDOM_Legacy
  shadingModel=MSM_Unlit
  usedShadingModels=Unlit
 }
 nodes [
  n1 = VertexNormalWS()
  n2 = Distance(A=WorldPosition().XYZ, B=CameraPositionWS())
  LinearInterpolate(A=@n1, B=Multiply(A=@n1, B=ScalarParameter() {defaultValue=10, parameterName=SizeScale}), Alpha=Clamp(None=Divide(A=@n2) {constB=15000}))
  n3 = Constant3Vector() {constant=<rgba 1 0 0 0>}
  n4 = Transform(None=@n3) {transformSourceType=TRANSFORMSOURCE_ParticleWorld}
  n5 = Transform(None=@n3) {transformSourceType=TRANSFORMSOURCE_Local}
  n6 = MaterialFunctionCall(
   Axis=StaticSwitchParameter(True=ShadingPathSwitch(Default=@n4, Mobile=@n5) {desc=Migrated from FeatureLevelSwitch}, False=@n5) {defaultValue=true, parameterName=UseWithMeshParticle}
   Exponent=ScalarParameter() {defaultValue=3, parameterName=FresnelExp}
  ) {materialFunction=&AxisAlignedFresnel}
  n7 = LinearInterpolate(
   A=VectorParameter() {parameterName=OutsideColor, defaultValue=<rgba 0.1322 0.0366 0.61 1>}.RGB
   B=VectorParameter() {parameterName=InsideColor, defaultValue=<rgba 33.0135 2.0465 50 1>}.RGB
   Alpha=@n6.Fresnel
  )
  n8 = ShadingPathSwitch(Default=@n7, Deferred=@n7, Mobile=VectorParameter() {parameterName=MobileColor, defaultValue=<rgba 33.0135 2.0465 50 1>}.RGB) {desc=Migrated from FeatureLevelSwitch}
  n9 = ParticleColor()
  n10 = StaticSwitchParameter(True=Multiply(A=@n8, B=@n9.RGB), False=@n8) {parameterName=Multiply Particle Color}
  n11 = Multiply(A=@n10) {constB=0.1}
  n12 = Multiply(A=ScalarParameter() {defaultValue=1, parameterName=Day_Intensity}, B=@n10) {constB=0.5}
  MaterialFunctionCall(Emissive=StaticSwitchParameter(
    True=MaterialFunctionCall(Color_Morning=@n11, Color_Day=@n12, Color_Evening=@n12, Color_Night=@n11) {materialFunction=&Color_Time_of_Day}.Result
    False=@n10
   ) {parameterName=Color_Time_of_Day}) {materialFunction=&MF_FX_EmissiveAdjustments}
  n13 = ScalarParameter() {defaultValue=50, parameterName=Near Camera Distance}
  n14 = ScalarParameter() {defaultValue=1.5, parameterName=AlphaFresnelPower}
  n15 = DepthFade(Opacity=Multiply(A=Power(Base=@n6.Fresnel, Exp=@n14), B=@n9.A), FadeDistance=ScalarParameter() {defaultValue=50, parameterName=FadeDistance})
  n16 = ShadingPathSwitch(Default=@n15, Deferred=@n15, Mobile=@n9.A) {desc=Migrated from FeatureLevelSwitch}
  n17 = MaterialFunctionCall() {materialFunction=&LinearGradient}
  n18 = Multiply(A=OneMinus(None=@n17.VGradient), B=@n17.VGradient)
  n19 = OneMinus(None=MaterialFunctionCall() {materialFunction=&LinearGradient}.VGradient)
  n20 = ScalarParameter() {parameterName=TopFade}
  n21 = Multiply(
   A=MaterialFunctionCall(In=Multiply(A=@n18) {constB=8}, Contrast=ScalarParameter() {parameterName=EndFade}) {materialFunction=&CheapContrast}.Result
   B=Clamp(None=Power(Base=@n19, Exp=@n20))
  )
  Multiply(
   A=Multiply(
    A=MaterialFunctionCall(Distance=@n13) {materialFunction=&Near_Camera_Fade}.Result
    B=Multiply(A=@n16, B=@n21)
   )
   B=ScalarParameter() {defaultValue=1, parameterName=OverallOpacity}
  )
 ]
}
