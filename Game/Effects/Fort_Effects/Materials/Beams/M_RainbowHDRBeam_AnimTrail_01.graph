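# Unlit translucent, two-sided material for Niagara ribbon and beam trails that glow as an animated HDR rainbow.
# A sine-warped rainbow gradient, tiled, scrolled and desaturated by dynamic parameters, is scaled by time-of-day brightness and particle color, then HDR-remapped.
# Opacity combines noise-broken length fades, two opposing scrolling streak layers and a soft width mask, times particle alpha.
# Depth and camera fades apply except on the mobile shading path.
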Material /Game/Effects/Fort_Effects/Materials/Beams/M_RainbowHDRBeam_AnimTrail_01
graph {
 properties {
  bUsedWithBeamTrails=true
  bUsedWithNiagaraRibbons=true
  blendMode=BLEND_Translucent
  pixelDepthOffsetMode=PDOM_Legacy
  shadingModel=MSM_Unlit
  twoSided=true
  usedShadingModels=Unlit
 }
 nodes [
  n1 = ScalarParameter() {defaultValue=1, parameterName=Brightness_Morning}
  n2 = ScalarParameter() {defaultValue=1.4, parameterName=Brightness_Day}
  n3 = ScalarParameter() {defaultValue=1, parameterName=Brightness_Evening}
  n4 = ScalarParameter() {defaultValue=0.65, parameterName=Brightness_Night}
  n5 = MaterialFunctionCall(XY=TextureCoordinate() {coordinateIndex=1}) {materialFunction=&Swizzle}
  n6 = ComponentMask(None=Sine(None=Multiply(A=@n5.YX) {constB=0.5})) {r=true}
  n7 = DynamicParameter() {paramNames [ColorScroll_Y, ColorTiling, Desaturate, Param4]}
  n8 = TextureSample(UVs=Add(A=Multiply(A=@n6, B=@n7.ColorTiling), B=@n7.ColorScroll_Y)) {texture=/Game/Global/Textures/Default/Masks_and_Gradients/T_Grad_Rainbow}
  n9 = ParticleColor()
  n10 = Multiply(
   A=Multiply(
    A=MaterialFunctionCall(Color_Morning=@n1, Color_Day=@n2, Color_Evening=@n3, Color_Night=@n4) {materialFunction=&Color_Time_of_Day}.Result
    B=Desaturation(None=@n8.RGB, Fraction=@n7.Desaturate)
   )
   B=@n9.RGB
  )
  MaterialFunctionCall(Emissive=MaterialFunctionCall(LDR Input=Clamp(None=@n10), HDR Tint=@n10) {materialFunction=&DeriveHDRfromLDR}.RemappedRange) {materialFunction=&MF_FX_EmissiveAdjustments}
  Reroute(None=@n9.A) {desc=Old wire}
  n11 = TextureSample(UVs=Panner(Coordinate=TextureCoordinate() {uTiling=0.2}) {bFractionalPart=true, speedX=-0.65}) {texture=&T_NoisePack_01}
  n12 = ComponentMask(None=TextureCoordinate()) {r=true}
  n13 = Power(Base=Add(A=ConstantBiasScale(None=@n11.R) {bias=-0.5, scale=0.2}, B=OneMinus(None=@n12))) {constExponent=6}
  n14 = Clamp(None=Multiply(A=@n12) {constB=8})
  n15 = ScalarParameter() {defaultValue=0.2, parameterName=Fine Streaks Opacity A}
  n16 = ScalarParameter() {defaultValue=0.3, parameterName=Fine Streaks Opacity B}
  n17 = MaterialFunctionCall(Float2=Panner(Coordinate=TextureCoordinate() {vTiling=3.464}) {speedY=0.976}) {materialFunction=&BreakOutFloat2Components}
  n18 = TextureSample(UVs=MaterialFunctionCall(X=@n17.G, Y=@n17.G) {materialFunction=&MakeFloat2}.Result) {texture=/Game/Effects/Fort_Effects/Textures/Outlander/T_InTheZone_LightRays_01}
  n19 = MaterialFunctionCall(Float2=Panner(Coordinate=TextureCoordinate() {vTiling=4.4362}) {speedY=-0.7616}) {materialFunction=&BreakOutFloat2Components}
  n20 = TextureSample(UVs=MaterialFunctionCall(X=@n19.G, Y=@n19.G) {materialFunction=&MakeFloat2}.Result) {texture=&T_InTheZone_LightRays_01}
  n21 = Divide(A=Add(A=Power(Base=@n18.R) {constExponent=1.5}, B=Power(Base=@n20.G)))
  n22 = Constant() {r=0.5}
  n23 = MaterialFunctionCall(Float2=TextureCoordinate()) {materialFunction=&BreakOutFloat2Components}
  n24 = Multiply(
   A=Multiply(A=Multiply(A=@n13, B=@n14), B=LinearInterpolate(A=@n15, B=@n16, Alpha=@n21) {constA=1})
   B=SphereMask(
    A=AppendVector(A=@n22, B=@n23.G)
    B=@n22
    Radius=ScalarParameter() {defaultValue=0.45, parameterName=Mask Radius}
    Hardness=ScalarParameter() {defaultValue=0.25, parameterName=Mask Hardness}
   )
  )
  n25 = DepthFade(Opacity=@n24) {fadeDistanceDefault=5}
  n26 = MaterialFunctionCall() {materialFunction=&CameraDepthFade}
  n27 = Multiply(A=Multiply(A=@n9.A, B=Multiply(A=@n25, B=@n26.Result)))
  ShadingPathSwitch(Default=@n27, Deferred=@n27, Mobile=Multiply(A=@n9.A, B=@n24)) {desc=Migrated from FeatureLevelSwitch}
 ]
}
